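# Physics asset for the F_MED_Blue_Flames female player cosmetic body, covering its dynamic boot flaps.
# Each side has a capsule body on the flap bone and a smaller capsule on its MSTR bone.
# A constraint joins each flap to its MSTR bone, with tight limits on one swing axis and twist and a wider range on the other swing axis, so the flap moves mostly like a hinge.
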PhysicsAsset /BRCosmetics/Characters/Player/Female/Medium/Bodies/F_MED_Blue_Flames/Meshes/F_MED_Blue_Flames_Dyn
bodies {
 dyn_boot_flap_l [{shapeType=Capsule, length=2, radius=2, center=<0.1 0 0>, rotation=<rot 90 0 0>}]
 dyn_boot_flap_l_MSTR [{shapeType=Capsule, length=0.5, radius=0.5, rotation=<rot 90 0 0>}]
 dyn_boot_flap_r [{shapeType=Capsule, length=2, radius=2, center=<0.1 0 0>, rotation=<rot 90 0 0>}]
 dyn_boot_flap_r_MSTR [{shapeType=Capsule, length=0.5, radius=0.5, rotation=<rot 90 0 0>}]
}
constraints [
 {bone1Name=dyn_boot_flap_l, bone2Name=dyn_boot_flap_l_MSTR, swing1LimitDegrees=5, swing2LimitDegrees=45, twistLimitDegrees=5}
 {bone1Name=dyn_boot_flap_r, bone2Name=dyn_boot_flap_r_MSTR, swing1LimitDegrees=5, swing2LimitDegrees=45, twistLimitDegrees=5}
]
