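# Physics asset for the left forearm hose bones on the F_MED_Blue_Flames female player body, giving them collision shapes and joints for physical motion.
# Two equally sized capsules cover the hose bones, and a smaller capsule on lowerarm_twist_02_l serves as their anchor.
# Each hose bone is constrained directly to that twist bone rather than chained to the other, with moderate swing limit angles and a much tighter twist angle.
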PhysicsAsset /BRCosmetics/Characters/Player/Female/Medium/Bodies/F_MED_Blue_Flames/Meshes/F_MED_Blue_Flames_Dyn_arm_l
bodies {
 dyn_forearm_hose_01_l [{shapeType=Capsule, length=2, radius=2, center=<0.2 0 0>, rotation=<rot 90 0 0>}]
 dyn_forearm_hose_02_l [{shapeType=Capsule, length=2, radius=2, center=<0.2 0 0>, rotation=<rot 90 0 0>}]
 lowerarm_twist_02_l [{shapeType=Capsule, length=0.5, radius=0.5, rotation=<rot 90 0 0>}]
}
constraints [
 {bone1Name=dyn_forearm_hose_02_l, bone2Name=lowerarm_twist_02_l, swing1LimitDegrees=45, swing2LimitDegrees=45, twistLimitDegrees=5}
 {bone1Name=dyn_forearm_hose_01_l, bone2Name=lowerarm_twist_02_l, swing1LimitDegrees=45, swing2LimitDegrees=45, twistLimitDegrees=5}
]
