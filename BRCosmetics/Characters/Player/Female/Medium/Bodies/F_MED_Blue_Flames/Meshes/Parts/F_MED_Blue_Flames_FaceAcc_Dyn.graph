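# Physics asset for the Blue Flames female medium body's face accessory, defining dynamic hair that hangs from the head.
# Box bodies form left, middle and right chains at front and back, with tight twist and direction-biased swing limits.
# Free-angular constraints cross-link the back middle chain to both side chains at their second bones.
# Capsules on head, neck, upper spine and clavicles form the body collision shapes, joined by free-angular constraints.
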PhysicsAsset /BRCosmetics/Characters/Player/Female/Medium/Bodies/F_MED_Blue_Flames/Meshes/Parts/F_MED_Blue_Flames_FaceAcc_Dyn
bodies {
 clavicle_l [{shapeType=Capsule, length=10.2525, radius=8.4568, center=<5.3274 0.8462 -1.9968>, rotation=<rot 90 0 -7.164>}]
 clavicle_r [{shapeType=Capsule, length=10.2525, radius=8.4568, center=<-5.3274 -0.8462 1.9968>, rotation=<rot 90 0 -7.164>}]
 dyn_hair_bk_l_01 [{shapeType=Box, extentX=2.2894, extentY=6.1875, extentZ=1.625, center=<1 0 0>, rotation=<rot 0 0 27.08>}]
 dyn_hair_bk_l_02 [{shapeType=Box, extentX=7.875, extentY=10.2077, extentZ=1.625, center=<3.6153 0 -0.0002>}]
 dyn_hair_bk_l_03 [{shapeType=Box, extentX=8.1875, extentY=10.2077, extentZ=1.625, center=<4.1311 0 -0.0002>, rotation=<rot -0.0001 0 -13.719>}]
 dyn_hair_bk_m_01 [{shapeType=Box, extentX=5.9375, extentY=8.6884, extentZ=1.625, center=<2.9628 0 -0.0002>}]
 dyn_hair_bk_m_02 [{shapeType=Box, extentX=7.5, extentY=8.6884, extentZ=1.625, center=<4.0754 0 -0.0002>}]
 dyn_hair_bk_m_03 [{shapeType=Box, extentX=6.3125, extentY=8.6884, extentZ=3.5, center=<3.4561 0 -0.0002>}]
 dyn_hair_bk_m_04 [{shapeType=Box, extentX=6, extentY=8.6884, extentZ=3.5, center=<2.9628 0 -0.0002>}]
 dyn_hair_bk_m_05 [{shapeType=Box, extentX=7.8125, extentY=8.6884, extentZ=3.5, center=<3.9938 0 -0.0002>}]
 dyn_hair_bk_r_01 [{shapeType=Box, extentX=2.2894, extentY=6.1875, extentZ=1.625, center=<1 0 0>, rotation=<rot 0 0 -27.08>}]
 dyn_hair_bk_r_02 [{shapeType=Box, extentX=7.875, extentY=10.2077, extentZ=1.625, center=<3.6153 0 -0.0002>}]
 dyn_hair_bk_r_03 [{shapeType=Box, extentX=8.1875, extentY=10.2077, extentZ=1.625, center=<4.1311 0 -0.0002>, rotation=<rot 0 0 14.0378>}]
 dyn_hair_fr_l_01 [{shapeType=Box, extentX=3.875, extentY=3.125, extentZ=1.625, center=<1.8946 0 -0.0003>}]
 dyn_hair_fr_l_02 [{shapeType=Box, extentX=3.5, extentY=3.125, extentZ=1.625, center=<2.187 -0.0001 -0.0002>}]
 dyn_hair_fr_m_01 [{shapeType=Box, extentX=3.875, extentY=3.125, extentZ=1.625, center=<2.1156 0 -0.0002>}]
 dyn_hair_fr_m_02 [{shapeType=Box, extentX=3.875, extentY=3.125, extentZ=1.625, center=<1.9415 0 -0.0002>}]
 dyn_hair_fr_m_03 [{shapeType=Box, extentX=2.125, extentY=3.125, extentZ=1.625, center=<1.3366 0 -0.0001>}]
 dyn_hair_fr_r_01 [{shapeType=Box, extentX=3.875, extentY=3.125, extentZ=1.625, center=<2.0779 0 -0.0002>}]
 dyn_hair_fr_r_02 [{shapeType=Box, extentX=3.875, extentY=3.125, extentZ=1.625, center=<2.2199 0 -0.0002>}]
 head [{shapeType=Capsule, length=5.5, radius=7.1875, center=<2.1343 1.3372 0>, rotation=<rot 90 0 0>}]
 neck_02 [{shapeType=Capsule, length=3.125, radius=5.0312, center=<-1.9925 0 0>, rotation=<rot 90 0 0>}]
 spine_04 [{shapeType=Capsule, length=8.4896, radius=9.75, center=<3.3725 -1.0035 0>}]
 spine_05 [{shapeType=Capsule, length=3.4062, radius=9.75, center=<6.8721 2.3603 0>}]
}
constraints [
 {bone1Name=dyn_hair_bk_l_01, bone2Name=head, swing1LimitDegrees=10, swing2LimitDegrees=45, twistLimitDegrees=5}
 {bone1Name=dyn_hair_bk_l_02, bone2Name=dyn_hair_bk_l_01, swing1LimitDegrees=45, swing2LimitDegrees=45, twistLimitDegrees=5}
 {bone1Name=dyn_hair_bk_l_03, bone2Name=dyn_hair_bk_l_02, swing1LimitDegrees=45, swing2LimitDegrees=45, twistLimitDegrees=5}
 {bone1Name=dyn_hair_bk_m_01, bone2Name=head, swing1LimitDegrees=5, swing2LimitDegrees=45, twistLimitDegrees=5}
 {bone1Name=dyn_hair_bk_m_02, bone2Name=dyn_hair_bk_m_01, swing1LimitDegrees=5, swing2LimitDegrees=45, twistLimitDegrees=5}
 {bone1Name=dyn_hair_bk_m_03, bone2Name=dyn_hair_bk_m_02, swing1LimitDegrees=5, swing2LimitDegrees=45, twistLimitDegrees=5}
 {bone1Name=dyn_hair_bk_m_04, bone2Name=dyn_hair_bk_m_03, swing1LimitDegrees=10, swing2LimitDegrees=45, twistLimitDegrees=5}
 {bone1Name=dyn_hair_bk_m_05, bone2Name=dyn_hair_bk_m_04, swing1LimitDegrees=10, swing2LimitDegrees=45, twistLimitDegrees=5}
 {bone1Name=dyn_hair_fr_l_01, bone2Name=head, swing1LimitDegrees=45, swing2LimitDegrees=5, twistLimitDegrees=5}
 {bone1Name=dyn_hair_fr_l_02, bone2Name=dyn_hair_fr_l_01, swing1LimitDegrees=45, swing2LimitDegrees=5, twistLimitDegrees=5}
 {bone1Name=dyn_hair_fr_m_01, bone2Name=head, swing1LimitDegrees=45, swing2LimitDegrees=5, twistLimitDegrees=5}
 {bone1Name=dyn_hair_fr_m_02, bone2Name=dyn_hair_fr_m_01, swing1LimitDegrees=45, swing2LimitDegrees=5, twistLimitDegrees=5}
 {bone1Name=dyn_hair_fr_m_03, bone2Name=dyn_hair_fr_m_02, swing1LimitDegrees=45, swing2LimitDegrees=5, twistLimitDegrees=5}
 {bone1Name=dyn_hair_fr_r_01, bone2Name=head, swing1LimitDegrees=45, swing2LimitDegrees=5, twistLimitDegrees=5}
 {bone1Name=dyn_hair_fr_r_02, bone2Name=dyn_hair_fr_r_01, swing1LimitDegrees=45, swing2LimitDegrees=5, twistLimitDegrees=5}
 {bone1Name=dyn_hair_bk_r_01, bone2Name=head, swing1LimitDegrees=10, swing2LimitDegrees=45, twistLimitDegrees=5}
 {bone1Name=dyn_hair_bk_r_02, bone2Name=dyn_hair_bk_r_01, swing1LimitDegrees=45, swing2LimitDegrees=45, twistLimitDegrees=5}
 {bone1Name=dyn_hair_bk_r_03, bone2Name=dyn_hair_bk_r_02, swing1LimitDegrees=45, swing2LimitDegrees=45, twistLimitDegrees=5}
 {
  bone1Name=head
  bone2Name=neck_02
  swing1LimitDegrees=45
  swing1Motion=Free
  swing2LimitDegrees=45
  swing2Motion=Free
  twistLimitDegrees=45
  twistMotion=Free
 }
 {
  bone1Name=clavicle_l
  bone2Name=spine_05
  swing1LimitDegrees=45
  swing1Motion=Free
  swing2LimitDegrees=45
  swing2Motion=Free
  twistLimitDegrees=45
  twistMotion=Free
 }
 {
  bone1Name=clavicle_r
  bone2Name=spine_05
  swing1LimitDegrees=45
  swing1Motion=Free
  swing2LimitDegrees=45
  swing2Motion=Free
  twistLimitDegrees=45
  twistMotion=Free
 }
 {
  bone1Name=neck_02
  bone2Name=spine_05
  swing1LimitDegrees=45
  swing1Motion=Free
  swing2LimitDegrees=45
  swing2Motion=Free
  twistLimitDegrees=45
  twistMotion=Free
 }
 {
  bone1Name=dyn_hair_bk_m_02
  bone2Name=dyn_hair_bk_r_02
  swing1LimitDegrees=45
  swing1Motion=Free
  swing2LimitDegrees=45
  swing2Motion=Free
  twistLimitDegrees=45
  twistMotion=Free
 }
 {
  bone1Name=dyn_hair_bk_m_02
  bone2Name=dyn_hair_bk_l_02
  swing1LimitDegrees=45
  swing1Motion=Free
  swing2LimitDegrees=45
  swing2Motion=Free
  twistLimitDegrees=45
  twistMotion=Free
 }
 {
  bone1Name=spine_05
  bone2Name=spine_04
  swing1LimitDegrees=45
  swing1Motion=Free
  swing2LimitDegrees=45
  swing2Motion=Free
  twistLimitDegrees=45
  twistMotion=Free
 }
]
