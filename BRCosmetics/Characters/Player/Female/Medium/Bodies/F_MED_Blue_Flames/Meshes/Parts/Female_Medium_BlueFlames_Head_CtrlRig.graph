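# Control Rig for the Blue Flames female medium head that applies corrective offsets and scaling to the eye and upper-lid bones.
# Upper-lid rotation relative to each eye drives lid offset and squash, eye flattening and lid narrowing; eye rotation relative to faceAttach drives further eye scaling.
# Clamped remaps of Euler angles weight a chain of ModifyTransforms nodes.
# Left and right lid-narrowing remaps are not mirrored, using different ranges and directions.
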ControlRigBlueprint /BRCosmetics/Characters/Player/Female/Medium/Bodies/F_MED_Blue_Flames/Meshes/Parts/Female_Medium_BlueFlames_Head_CtrlRig
graphs [
 RigVMFunctionLibrary {links [], nodes []}
 RigVMModel {nodes [
   RigUnit_BeginExecution = &RigUnit_BeginExecution()
   n1 = &RigVMFunction_MathFloatRemap(
    SourceMaximum=40
    SourceMinimum=0
    TargetMinimum=0
    bClamp=true
    Value=&RigVMFunction_MathQuaternionToEuler(
     RotationOrder=ZYX
     Value=&RigUnit_GetRelativeTransformForItem(Child=(Type=Bone,Name="L_eye_lid_upper_mid"), Parent=(Type=Bone,Name="L_eye")).RelativeTransform.Rotation
    ).Result.X
   )
   &RigUnit_ModifyTransforms(
    ItemToModify=((Item=(Type=Bone,Name="L_eye_lid_upper_mid"),Transform=(Rotation=(X=0,Y=0,Z=0,W=1),Translation=(X=0,Y=0.6,Z=-0.2),Scale3D=(X=0.9,Y=0.7,Z=0.4))))
    Weight=@n1.Result
    ExecutePin=@RigUnit_BeginExecution.ExecutePin
   )
   > &RigUnit_ModifyTransforms(
    ItemToModify=((Item=(Type=Bone,Name="L_eye"),Transform=(Rotation=(X=0,Y=0,Z=0,W=1),Translation=(X=0,Y=0.2,Z=0),Scale3D=(X=1,Y=0.2,Z=1))))
    Weight=@n1.Result
   )
   n2 = &RigVMFunction_MathFloatRemap(
    SourceMaximum=40
    SourceMinimum=0
    TargetMinimum=0
    bClamp=true
    Value=&RigVMFunction_MathQuaternionToEuler(
     RotationOrder=ZYX
     Value=&RigUnit_GetRelativeTransformForItem(Child=(Type=Bone,Name="R_eye_lid_upper_mid"), Parent=(Type=Bone,Name="R_eye")).RelativeTransform.Rotation
    ).Result.X
   )
   > &RigUnit_ModifyTransforms(
    ItemToModify=((Item=(Type=Bone,Name="R_eye_lid_upper_mid"),Transform=(Rotation=(X=0,Y=0,Z=0,W=1),Translation=(X=0,Y=0.6,Z=-0.2),Scale3D=(X=0.9,Y=0.7,Z=0.4))))
    Weight=@n2.Result
   )
   > &RigUnit_ModifyTransforms(
    ItemToModify=((Item=(Type=Bone,Name="R_eye"),Transform=(Rotation=(X=0,Y=0,Z=0,W=1),Translation=(X=0,Y=0.2,Z=0),Scale3D=(X=1,Y=0.2,Z=1))))
    Weight=@n2.Result
   )
   n3 = &RigUnit_GetRelativeTransformForItem(Child=(Type=Bone,Name="L_eye_lid_upper_mid"), Parent=(Type=Bone,Name="L_eye"))
   > &RigUnit_ModifyTransforms(
    ItemToModify=((Item=(Type=Bone,Name="L_eye_lid_upper_mid"),Transform=(Rotation=(X=0,Y=0,Z=0,W=1),Translation=(X=0,Y=0,Z=0),Scale3D=(X=0.75,Y=1,Z=1))))
    WeightMinimum=0
    Weight=&RigVMFunction_MathFloatRemap(
     SourceMaximum=0
     SourceMinimum=-15
     TargetMaximum=0
     TargetMinimum=1
     bClamp=true
     Value=&RigVMFunction_MathQuaternionToEuler(RotationOrder=ZYX, Value=@n3.RelativeTransform.Rotation).Result.Z
    ).Result
   )
   n4 = &RigUnit_GetRelativeTransformForItem(Child=(Type=Bone,Name="R_eye_lid_upper_mid"), Parent=(Type=Bone,Name="R_eye"))
   > &RigUnit_ModifyTransforms(
    ItemToModify=((Item=(Type=Bone,Name="R_eye_lid_upper_mid"),Transform=(Rotation=(X=0,Y=0,Z=0,W=1),Translation=(X=0,Y=0,Z=0),Scale3D=(X=0.75,Y=1,Z=1))))
    WeightMinimum=0
    Weight=&RigVMFunction_MathFloatRemap(
     SourceMaximum=15
     SourceMinimum=-15
     TargetMinimum=0
     bClamp=true
     Value=&RigVMFunction_MathQuaternionToEuler(RotationOrder=ZYX, Value=@n4.RelativeTransform.Rotation).Result.Z
    ).Result
   )
   n5 = &RigUnit_GetRelativeTransformForItem(Child=(Type=Bone,Name="L_eye"), Parent=(Type=Bone,Name="faceAttach"))
   > &RigUnit_ModifyTransforms(
    ItemToModify=((Item=(Type=Bone,Name="L_eye"),Transform=(Rotation=(X=0,Y=0,Z=0,W=1),Translation=(X=0,Y=0,Z=0),Scale3D=(X=0.9,Y=0.7,Z=1))))
    WeightMinimum=0
    Weight=&RigVMFunction_MathFloatRemap(
     SourceMaximum=0
     SourceMinimum=-25
     TargetMaximum=0
     TargetMinimum=1
     bClamp=true
     Value=&RigVMFunction_MathQuaternionToEuler(RotationOrder=ZYX, Value=@n5.RelativeTransform.Rotation).Result.X
    ).Result
   )
   n6 = &RigUnit_GetRelativeTransformForItem(Child=(Type=Bone,Name="R_eye"), Parent=(Type=Bone,Name="faceAttach"))
   > &RigUnit_ModifyTransforms(
    ItemToModify=((Item=(Type=Bone,Name="R_eye"),Transform=(Rotation=(X=0,Y=0,Z=0,W=1),Translation=(X=0,Y=0,Z=0),Scale3D=(X=0.9,Y=0.7,Z=1))))
    WeightMinimum=0
    Weight=&RigVMFunction_MathFloatRemap(
     SourceMaximum=0
     SourceMinimum=-25
     TargetMaximum=0
     TargetMinimum=1
     bClamp=true
     Value=&RigVMFunction_MathQuaternionToEuler(RotationOrder=ZYX, Value=@n6.RelativeTransform.Rotation).Result.X
    ).Result
   )
  ]}
]
hierarchy [
 root {loc=<0 0 0>, rot=<0 0 0 1>}
 attach {parent=root, loc=<0 0 0>, rot=<0 0 0 1>}
 pelvis {parent=root, loc=<0 -0.5301 98.0571>, rot=<0.0127 0.707 -0.0127 -0.707>}
 spine_01 {parent=pelvis, loc=<6.1208 -0.9328 0>, rot=<0 0 0.034 -0.9994>}
 spine_02 {parent=spine_01, loc=<6.5996 -0.2095 0>, rot=<0 0 0.0238 -0.9997>}
 spine_03 {parent=spine_02, loc=<6.4914 -0.1237 0>, rot=<0 0 0.0149 -0.9999>}
 spine_04 {parent=spine_03, loc=<6.9598 0.0794 0>, rot=<0 0 -0.0132 -0.9999>}
 spine_05 {parent=spine_04, loc=<6.9987 0 0>, rot=<0 0 -0.0015 -1>}
 clavicle_l {parent=spine_05, loc=<5.4831 -0.1053 -3.2946>, rot=<0.0465 -0.7289 -0.009 -0.6829>}
 upperarm_l {parent=clavicle_l, loc=<9.8658 -1.2136 -0.4105>, rot=<0.0208 -0.486 -0.0242 -0.8734>}
 lowerarm_l {parent=upperarm_l, loc=<26.465 -0.8345 0.259>, rot=<0.0036 0.0161 -0.2298 -0.9731>}
 hand_l {parent=lowerarm_l, loc=<22.7355 -0.3203 -0.6105>, rot=<-0.6784 -0.0223 0.0735 -0.7307>}
 index_metacarpal_l {parent=hand_l, loc=<1.6535 0.0183 -2.96>, rot=<0.0158 -0.0124 -0.0173 -0.9996>}
 index_01_l {parent=index_metacarpal_l, loc=<8.3008 0.0357 -0.0524>, rot=<-0.0044 0.0485 0.0924 -0.9945>}
 index_02_l {parent=index_01_l, loc=<3.3587 -0.3004 0.0452>, rot=<-0.002 0.0027 0.3225 -0.9466>}
 index_03_l {parent=index_02_l, loc=<2.5578 -0.123 -0.0252>, rot=<-0.002 0.0067 0.2558 -0.9667>}
 middle_metacarpal_l {parent=hand_l, loc=<2.1856 -0.1725 -0.5984>, rot=<0.006 0.0168 -0.0329 -0.9993>}
 middle_01_l {parent=middle_metacarpal_l, loc=<7.3275 -0.2447 0.0129>, rot=<-0.0208 0.0042 0.0858 -0.9961>}
 middle_02_l {parent=middle_01_l, loc=<4.0123 0.1401 0.1175>, rot=<-0.0218 0.0258 0.3512 -0.9357>}
 middle_03_l {parent=middle_02_l, loc=<2.6077 0.0534 0.1373>, rot=<0.0082 0.0188 0.1708 -0.9851>}
 pinky_metacarpal_l {parent=hand_l, loc=<1.7616 -0.9753 3.3814>, rot=<-0.0397 0.0814 -0.069 -0.9935>}
 pinky_01_l {parent=pinky_metacarpal_l, loc=<7.1709 -0.4831 -0.1094>, rot=<-0.0654 -0.0035 0.2242 -0.9723>}
 pinky_02_l {parent=pinky_01_l, loc=<2.8274 0.0893 -0.1043>, rot=<0.0037 0.0011 0.2854 -0.9584>}
 pinky_03_l {parent=pinky_02_l, loc=<2.2046 0.0307 -0.0078>, rot=<-0.0002 0.0098 0.1445 -0.9895>}
 ring_metacarpal_l {parent=hand_l, loc=<2.2074 -0.3625 1.3638>, rot=<-0.0284 0.0722 -0.0458 -0.9959>}
 ring_01_l {parent=ring_metacarpal_l, loc=<7.0295 -0.6197 -0.1446>, rot=<-0.0489 -0.0252 0.1892 -0.9804>}
 ring_02_l {parent=ring_01_l, loc=<3.6174 0.4647 -0.1027>, rot=<0 0 0.2806 -0.9598>}
 ring_03_l {parent=ring_02_l, loc=<2.4152 0.114 0.0199>, rot=<0 0 0.1189 -0.9929>}
 thumb_01_l {parent=hand_l, loc=<1.3118 -0.3191 -2.459>, rot=<0.6546 -0.4042 0.0664 -0.6354>}
 thumb_02_l {parent=thumb_01_l, loc=<5.9336 -0.0448 0.0401>, rot=<-0.0037 0.0213 0.298 -0.9543>}
 thumb_03_l {parent=thumb_02_l, loc=<2.2223 0.3682 -0.0907>, rot=<0.0118 -0.013 0.0066 -0.9998>}
 weapon_l {parent=hand_l, loc=<-0.1974 -3.4111 0.3811>, rot=<0 0 0 1>}
 lowerarm_twist_01_l {parent=lowerarm_l, loc=<18.316 0 0>, rot=<0 0 0 1>}
 lowerarm_twist_02_l {parent=lowerarm_l, loc=<10.6707 0 0>, rot=<0 0 0 1>}
 upperarm_twist_01_l {parent=upperarm_l, loc=<7.2135 -0.1134 0>, rot=<0 0 0 1>}
 upperarm_twist_02_l {parent=upperarm_l, loc=<19.7998 -0.1134 0>, rot=<0 0 0 1>}
 clavicle_r {parent=spine_05, loc=<5.4831 -0.1053 3.2946>, rot=<0.7289 0.0465 -0.6829 0.009>}
 upperarm_r {parent=clavicle_r, loc=<-9.8658 1.2136 0.4104>, rot=<0.0208 -0.486 -0.0242 -0.8734>}
 lowerarm_r {parent=upperarm_r, loc=<-26.4651 0.8345 -0.259>, rot=<0.0036 0.0161 -0.2298 -0.9731>}
 hand_r {parent=lowerarm_r, loc=<-22.7356 0.3203 0.6106>, rot=<-0.6784 -0.0223 0.0735 -0.7307>}
 index_metacarpal_r {parent=hand_r, loc=<-1.6534 -0.0181 2.9601>, rot=<0.0158 -0.0124 -0.0173 -0.9996>}
 index_01_r {parent=index_metacarpal_r, loc=<-8.3008 -0.0357 0.0524>, rot=<-0.0044 0.0485 0.0924 -0.9945>}
 index_02_r {parent=index_01_r, loc=<-3.3589 0.3003 -0.0453>, rot=<-0.002 0.0027 0.3225 -0.9466>}
 index_03_r {parent=index_02_r, loc=<-2.5578 0.123 0.0252>, rot=<-0.002 0.0067 0.2558 -0.9667>}
 middle_metacarpal_r {parent=hand_r, loc=<-2.1856 0.1726 0.5984>, rot=<0.006 0.0168 -0.0329 -0.9993>}
 middle_01_r {parent=middle_metacarpal_r, loc=<-7.3273 0.2447 -0.0129>, rot=<-0.0208 0.0042 0.0858 -0.9961>}
 middle_02_r {parent=middle_01_r, loc=<-4.0124 -0.14 -0.1175>, rot=<-0.0218 0.0258 0.3512 -0.9357>}
 middle_03_r {parent=middle_02_r, loc=<-2.6077 -0.0533 -0.1373>, rot=<0.0082 0.0188 0.1708 -0.9851>}
 pinky_metacarpal_r {parent=hand_r, loc=<-1.7617 0.9753 -3.3814>, rot=<-0.0397 0.0814 -0.069 -0.9935>}
 pinky_01_r {parent=pinky_metacarpal_r, loc=<-7.1707 0.4832 0.1094>, rot=<-0.0654 -0.0035 0.2242 -0.9723>}
 pinky_02_r {parent=pinky_01_r, loc=<-2.8274 -0.0892 0.1043>, rot=<0.0037 0.0011 0.2854 -0.9584>}
 pinky_03_r {parent=pinky_02_r, loc=<-2.2045 -0.031 0.0078>, rot=<-0.0002 0.0098 0.1445 -0.9895>}
 ring_metacarpal_r {parent=hand_r, loc=<-2.2073 0.3627 -1.3638>, rot=<-0.0284 0.0722 -0.0458 -0.9959>}
 ring_01_r {parent=ring_metacarpal_r, loc=<-7.0297 0.6197 0.1446>, rot=<-0.0489 -0.0252 0.1892 -0.9804>}
 ring_02_r {parent=ring_01_r, loc=<-3.6171 -0.4648 0.1027>, rot=<0 0 0.2806 -0.9598>}
 ring_03_r {parent=ring_02_r, loc=<-2.4153 -0.1141 -0.0199>, rot=<0 0 0.1189 -0.9929>}
 thumb_01_r {parent=hand_r, loc=<-1.3119 0.3192 2.459>, rot=<0.6546 -0.4042 0.0664 -0.6354>}
 thumb_02_r {parent=thumb_01_r, loc=<-5.9335 0.0447 -0.0401>, rot=<-0.0037 0.0213 0.298 -0.9543>}
 thumb_03_r {parent=thumb_02_r, loc=<-2.2223 -0.3682 0.0907>, rot=<0.0118 -0.013 0.0066 -0.9998>}
 weapon_r {parent=hand_r, loc=<0.1975 3.4112 -0.3811>, rot=<0 0 0 -1>}
 lowerarm_twist_01_r {parent=lowerarm_r, loc=<-18.3161 0 0>, rot=<0 0 0 1>}
 lowerarm_twist_02_r {parent=lowerarm_r, loc=<-10.6707 0 0>, rot=<0 0 0 1>}
 upperarm_twist_01_r {parent=upperarm_r, loc=<-7.2135 0.1134 0>, rot=<0 0 0 1>}
 upperarm_twist_02_r {parent=upperarm_r, loc=<-19.7998 0.1134 0>, rot=<0 0 0 1>}
 neck_01 {parent=spine_05, loc=<11.3138 0.1529 0>, rot=<0 0 -0.2003 -0.9797>}
 neck_02 {parent=neck_01, loc=<4.5627 0 0>, rot=<0 0 0.0504 -0.9987>}
 head {parent=neck_02, loc=<5.3914 0 0>, rot=<0 0 0.115 -0.9934>}
 faceAttach {parent=head, loc=<-3.9918 -0.3948 0.0161>, rot=<0.0082 -0.7071 -0.0082 -0.7071>, scale=<0.89 0.89 0.89>}
 C_brow_mid {parent=faceAttach, loc=<-0.0049 10.2838 8.5143>, rot=<0 0.0365 -0.0072 -0.9993>}
 L_brow_mid {parent=faceAttach, loc=<3.3552 9.4462 9.9853>, rot=<-0.0002 -0.0049 0.2553 -0.9669>}
 L_brow_outer {parent=faceAttach, loc=<5.7183 7.1955 9.8976>, rot=<0.1642 0.2029 0.3881 -0.8839>}
 L_cheek_inner {parent=faceAttach, loc=<4.5589 8.4684 3.708>, rot=<-0.0464 -0.2084 0.3262 -0.9209>}
 C_lip_upper_mid {parent=faceAttach, loc=<0.0239 10.5579 1.7231>, rot=<0.2225 0.001 -0.0004 -0.9749>}
 L_lip_upper_outer {parent=faceAttach, loc=<1.7169 9.7492 1.7073>, rot=<0.041 -0.0585 0.2573 -0.9637>}
 L_lip_corner {parent=faceAttach, loc=<2.4551 9.0395 1.7277>, rot=<0.0029 0.008 -0.0146 -0.9999>}
 C_nose_bridge {parent=faceAttach, loc=<0.0063 11.78 5.4358>, rot=<-0.2218 -0.0057 -0.0016 -0.9751>}
 R_brow_mid {parent=faceAttach, loc=<-3.319 9.4462 9.9853>, rot=<-0.0002 0.0049 -0.2553 -0.9669>}
 R_brow_outer {parent=faceAttach, loc=<-5.6865 7.1929 9.8749>, rot=<0.0927 -0.1128 -0.416 -0.8976>}
 R_cheek_inner {parent=faceAttach, loc=<-4.5102 8.4776 3.7003>, rot=<-0.0476 0.2116 -0.3259 -0.9202>}
 R_lip_corner {parent=faceAttach, loc=<-2.4554 9.0104 1.7194>, rot=<0.003 -0.008 0.0146 -0.9999>}
 C_jaw {parent=faceAttach, loc=<0.0112 2.9297 1.4813>, rot=<0 0 0 1>}
 L_lip_lower_outer {parent=C_jaw, loc=<1.6776 6.7407 0.2572>, rot=<0.0082 0.0184 0.2857 -0.9581>}
 C_lip_lower_mid {parent=C_jaw, loc=<0.0085 7.6235 0.0467>, rot=<0 -0.019 -0.004 -0.9998>}
 R_lip_lower_outer {parent=C_jaw, loc=<-1.7283 6.8319 0.1159>, rot=<-0.0003 0.0276 -0.2499 -0.9679>}
 teeth_lower {parent=C_jaw, loc=<-0.0215 5.118 -1.5516>, rot=<0 0 0 1>}
 tongue {parent=C_jaw, loc=<-0.0026 1.5714 -0.8898>, rot=<-0.1568 0 0 -0.9876>}
 R_eye {parent=faceAttach, loc=<-3.2503 7.1846 8.5766>, rot=<-0.0074 0 0 -1>}
 R_eye_lid_upper_mid {parent=R_eye, loc=<-0.3271 0.3742 -1.0343>, rot=<0.0522 -0.0033 -0.0637 -0.9966>, scale=<0.9989 1.0011 1>}
 R_eye_lid_lower_mid {parent=R_eye, loc=<-0.1691 0.9514 -0.0411>, rot=<0.0326 -0.0039 -0.1201 -0.9922>, scale=<0.997 1.003 1>}
 L_eye {parent=faceAttach, loc=<3.2866 7.1846 8.5766>, rot=<-0.0074 0 0 -1>}
 L_eye_lid_upper_mid {parent=L_eye, loc=<0.3457 0.3828 -1.0584>, rot=<0.0522 0.0035 0.0669 -0.9964>}
 L_eye_lid_lower_mid {parent=L_eye, loc=<0.2098 0.9639 -0.0844>, rot=<0.0326 0.0041 0.1254 -0.9916>}
 teeth_upper {parent=faceAttach, loc=<-0.0065 8.5629 2.6599>, rot=<0 0 0 1>}
 R_lip_upper_outer {parent=faceAttach, loc=<-1.7 9.8018 1.7226>, rot=<0.0402 0.0609 -0.2582 -0.9633>}
 thigh_l {parent=pelvis, loc=<-1.0855 -0.2109 -8.7053>, rot=<0.0009 0.0355 0.0155 -0.9993>}
 calf_l {parent=thigh_l, loc=<-42.5284 -0.0019 -0.7297>, rot=<0.0006 -0.0161 -0.0391 -0.9991>}
 calf_twist_01_l {parent=calf_l, loc=<-35.8483 -1.9197 -0.5794>, rot=<0.0237 0.0095 -0.0035 -0.9997>}
 calf_twist_02_l {parent=calf_l, loc=<-19.1546 -1.7943 -0.2638>, rot=<0.0237 0.0095 -0.0035 -0.9997>}
 foot_l {parent=calf_l, loc=<-46.7402 -2.5014 -0.592>, rot=<0.0427 -0.0238 0.0353 -0.9982>}
 ball_l {parent=foot_l, loc=<-3.5658 10.8066 -0.1544>, rot=<0.0001 -0.0001 -0.7063 -0.7079>}
 thigh_twist_01_l {parent=thigh_l, loc=<-27.952 0 0>, rot=<0.0013 0 0 -1>}
 thigh_r {parent=pelvis, loc=<-1.0853 -0.2109 8.7053>, rot=<0.0355 -0.0009 0.9993 0.0155>}
 calf_r {parent=thigh_r, loc=<42.5286 0.0019 0.7297>, rot=<0.0006 -0.0161 -0.0391 -0.9991>}
 calf_twist_01_r {parent=calf_r, loc=<35.8485 1.9197 0.5794>, rot=<0.0237 0.0095 -0.0035 -0.9997>}
 calf_twist_02_r {parent=calf_r, loc=<19.1546 1.7943 0.2638>, rot=<0.0237 0.0095 -0.0035 -0.9997>}
 foot_r {parent=calf_r, loc=<46.7403 2.5014 0.592>, rot=<0.0427 -0.0238 0.0353 -0.9982>}
 ball_r {parent=foot_r, loc=<3.5658 -10.8066 0.1544>, rot=<0.0001 -0.0001 -0.7063 -0.7079>}
 thigh_twist_01_r {parent=thigh_r, loc=<27.952 0 0>, rot=<0.0013 0 0 -1>}
 look_left_pose {type=Curve, loc=<0 0 0>, rot=<0 0 0 1>}
 look_right_pose {type=Curve, loc=<0 0 0>, rot=<0 0 0 1>}
 look_up_pose {type=Curve, loc=<0 0 0>, rot=<0 0 0 1>}
 look_down_pose {type=Curve, loc=<0 0 0>, rot=<0 0 0 1>}
 L_blink_pose {type=Curve, loc=<0 0 0>, rot=<0 0 0 1>}
 R_blink_pose {type=Curve, loc=<0 0 0>, rot=<0 0 0 1>}
 L_wide_pose {type=Curve, loc=<0 0 0>, rot=<0 0 0 1>}
 R_wide_pose {type=Curve, loc=<0 0 0>, rot=<0 0 0 1>}
 R_squint_inner_pose {type=Curve, loc=<0 0 0>, rot=<0 0 0 1>}
 R_squeeze_pose {type=Curve, loc=<0 0 0>, rot=<0 0 0 1>}
 L_squint_inner_pose {type=Curve, loc=<0 0 0>, rot=<0 0 0 1>}
 L_squeeze_pose {type=Curve, loc=<0 0 0>, rot=<0 0 0 1>}
 L_brow_up_pose {type=Curve, loc=<0 0 0>, rot=<0 0 0 1>}
 L_brow_down_pose {type=Curve, loc=<0 0 0>, rot=<0 0 0 1>}
 R_brow_up_pose {type=Curve, loc=<0 0 0>, rot=<0 0 0 1>}
 R_brow_down_pose {type=Curve, loc=<0 0 0>, rot=<0 0 0 1>}
 C_glabella_up_pose {type=Curve, loc=<0 0 0>, rot=<0 0 0 1>}
 C_glabella_down_pose {type=Curve, loc=<0 0 0>, rot=<0 0 0 1>}
 L_brow_squeeze_in_pose {type=Curve, loc=<0 0 0>, rot=<0 0 0 1>}
 R_brow_squeeze_in_pose {type=Curve, loc=<0 0 0>, rot=<0 0 0 1>}
 L_upper_lip_raiser_pose {type=Curve, loc=<0 0 0>, rot=<0 0 0 1>}
 R_upper_lip_raiser_pose {type=Curve, loc=<0 0 0>, rot=<0 0 0 1>}
 L_upper_lip_lower_pose {type=Curve, loc=<0 0 0>, rot=<0 0 0 1>}
 R_upper_lip_lower_pose {type=Curve, loc=<0 0 0>, rot=<0 0 0 1>}
 L_lower_lip_up_pose {type=Curve, loc=<0 0 0>, rot=<0 0 0 1>}
 R_lower_lip_up_pose {type=Curve, loc=<0 0 0>, rot=<0 0 0 1>}
 L_lower_lip_down_pose {type=Curve, loc=<0 0 0>, rot=<0 0 0 1>}
 R_lower_lip_down_pose {type=Curve, loc=<0 0 0>, rot=<0 0 0 1>}
 L_lip_corner_narrow_pose {type=Curve, loc=<0 0 0>, rot=<0 0 0 1>}
 L_lip_corner_wide_pose {type=Curve, loc=<0 0 0>, rot=<0 0 0 1>}
 R_lip_corner_narrow_pose {type=Curve, loc=<0 0 0>, rot=<0 0 0 1>}
 R_lip_corner_wide_pose {type=Curve, loc=<0 0 0>, rot=<0 0 0 1>}
 L_smile_pose {type=Curve, loc=<0 0 0>, rot=<0 0 0 1>}
 R_smile_pose {type=Curve, loc=<0 0 0>, rot=<0 0 0 1>}
 L_frown_pose {type=Curve, loc=<0 0 0>, rot=<0 0 0 1>}
 R_frown_pose {type=Curve, loc=<0 0 0>, rot=<0 0 0 1>}
 phoneme_oo_pose {type=Curve, loc=<0 0 0>, rot=<0 0 0 1>}
 phoneme_ch_pose {type=Curve, loc=<0 0 0>, rot=<0 0 0 1>}
 phoneme_mbp_pose {type=Curve, loc=<0 0 0>, rot=<0 0 0 1>}
 phoneme_fv_pose {type=Curve, loc=<0 0 0>, rot=<0 0 0 1>}
 mouth_up_pose {type=Curve, loc=<0 0 0>, rot=<0 0 0 1>}
 mouth_down_pose {type=Curve, loc=<0 0 0>, rot=<0 0 0 1>}
 C_nose_wrinkler_pose {type=Curve, loc=<0 0 0>, rot=<0 0 0 1>}
 L_cheek_up_pose {type=Curve, loc=<0 0 0>, rot=<0 0 0 1>}
 R_cheek_up_pose {type=Curve, loc=<0 0 0>, rot=<0 0 0 1>}
 jaw_open_pose {type=Curve, loc=<0 0 0>, rot=<0 0 0 1>}
 jaw_right_pose {type=Curve, loc=<0 0 0>, rot=<0 0 0 1>}
 jaw_left_pose {type=Curve, loc=<0 0 0>, rot=<0 0 0 1>}
 eye_convergence_pose {type=Curve, loc=<0 0 0>, rot=<0 0 0 1>}
 L_upperLidDown_pose {type=Curve, loc=<0 0 0>, rot=<0 0 0 1>}
 R_upperLidDown_pose {type=Curve, loc=<0 0 0>, rot=<0 0 0 1>}
 jaw_up_pose {type=Curve, loc=<0 0 0>, rot=<0 0 0 1>}
 jaw_forward_pose {type=Curve, loc=<0 0 0>, rot=<0 0 0 1>}
 jaw_back_pose {type=Curve, loc=<0 0 0>, rot=<0 0 0 1>}
 tongue_forward_pose {type=Curve, loc=<0 0 0>, rot=<0 0 0 1>}
 tongue_back_pose {type=Curve, loc=<0 0 0>, rot=<0 0 0 1>}
 tongue_up_pose {type=Curve, loc=<0 0 0>, rot=<0 0 0 1>}
 tongue_down_pose {type=Curve, loc=<0 0 0>, rot=<0 0 0 1>}
 base_pose {type=Curve, loc=<0 0 0>, rot=<0 0 0 1>}
 Hair_Scarf {type=Curve, loc=<0 0 0>, rot=<0 0 0 1>}
 Hair_Hat {type=Curve, loc=<0 0 0>, rot=<0 0 0 1>}
 Hair_Helmet {type=Curve, loc=<0 0 0>, rot=<0 0 0 1>}
 Hair_Glasses {type=Curve, loc=<0 0 0>, rot=<0 0 0 1>}
 Hair_Cap {type=Curve, loc=<0 0 0>, rot=<0 0 0 1>}
 Hair_Headsets {type=Curve, loc=<0 0 0>, rot=<0 0 0 1>}
 Hair_Hood {type=Curve, loc=<0 0 0>, rot=<0 0 0 1>}
 Face_Mask {type=Curve, loc=<0 0 0>, rot=<0 0 0 1>}
 HairStrands_Hat {type=Curve, loc=<0 0 0>, rot=<0 0 0 1>}
 Hat_Hair {type=Curve, loc=<0 0 0>, rot=<0 0 0 1>}
 Helmet_Hair {type=Curve, loc=<0 0 0>, rot=<0 0 0 1>}
 Hair_Muff {type=Curve, loc=<0 0 0>, rot=<0 0 0 1>}
 Hair_SpaceSuit {type=Curve, loc=<0 0 0>, rot=<0 0 0 1>}
 Hair_Hide {type=Curve, loc=<0 0 0>, rot=<0 0 0 1>}
 Head_SpaceSuit {type=Curve, loc=<0 0 0>, rot=<0 0 0 1>}
 Hair_Null {type=Curve, loc=<0 0 0>, rot=<0 0 0 1>}
 DisableFaceOverride {type=Curve, loc=<0 0 0>, rot=<0 0 0 1>}
 LiveLink {type=Curve, loc=<0 0 0>, rot=<0 0 0 1>}
 Hat_MESH {type=Curve, loc=<0 0 0>, rot=<0 0 0 1>}
 Null_MESH {type=Curve, loc=<0 0 0>, rot=<0 0 0 1>}
 Pose_0 {type=Curve, loc=<0 0 0>, rot=<0 0 0 1>}
 Hair_Mask {type=Curve, loc=<0 0 0>, rot=<0 0 0 1>}
 eye_convergence {type=Curve, loc=<0 0 0>, rot=<0 0 0 1>}
 headTransform {type=Curve, loc=<0 0 0>, rot=<0 0 0 1>}
 Pose_1 {type=Curve, loc=<0 0 0>, rot=<0 0 0 1>}
 Pose_2 {type=Curve, loc=<0 0 0>, rot=<0 0 0 1>}
 Pose_3 {type=Curve, loc=<0 0 0>, rot=<0 0 0 1>}
 Pose_4 {type=Curve, loc=<0 0 0>, rot=<0 0 0 1>}
 Pose_5 {type=Curve, loc=<0 0 0>, rot=<0 0 0 1>}
 Pose_6 {type=Curve, loc=<0 0 0>, rot=<0 0 0 1>}
 Pose_7 {type=Curve, loc=<0 0 0>, rot=<0 0 0 1>}
 Pose_8 {type=Curve, loc=<0 0 0>, rot=<0 0 0 1>}
 Pose_9 {type=Curve, loc=<0 0 0>, rot=<0 0 0 1>}
 Pose_10 {type=Curve, loc=<0 0 0>, rot=<0 0 0 1>}
 Pose_11 {type=Curve, loc=<0 0 0>, rot=<0 0 0 1>}
 Pose_12 {type=Curve, loc=<0 0 0>, rot=<0 0 0 1>}
 Pose_13 {type=Curve, loc=<0 0 0>, rot=<0 0 0 1>}
 Pose_14 {type=Curve, loc=<0 0 0>, rot=<0 0 0 1>}
 Pose_15 {type=Curve, loc=<0 0 0>, rot=<0 0 0 1>}
 Pose_16 {type=Curve, loc=<0 0 0>, rot=<0 0 0 1>}
 Pose_17 {type=Curve, loc=<0 0 0>, rot=<0 0 0 1>}
 Pose_18 {type=Curve, loc=<0 0 0>, rot=<0 0 0 1>}
 Pose_19 {type=Curve, loc=<0 0 0>, rot=<0 0 0 1>}
 Pose_20 {type=Curve, loc=<0 0 0>, rot=<0 0 0 1>}
 Pose_21 {type=Curve, loc=<0 0 0>, rot=<0 0 0 1>}
 Pose_22 {type=Curve, loc=<0 0 0>, rot=<0 0 0 1>}
 Pose_23 {type=Curve, loc=<0 0 0>, rot=<0 0 0 1>}
 Pose_24 {type=Curve, loc=<0 0 0>, rot=<0 0 0 1>}
 Pose_25 {type=Curve, loc=<0 0 0>, rot=<0 0 0 1>}
 Pose_26 {type=Curve, loc=<0 0 0>, rot=<0 0 0 1>}
 Pose_27 {type=Curve, loc=<0 0 0>, rot=<0 0 0 1>}
 Pose_28 {type=Curve, loc=<0 0 0>, rot=<0 0 0 1>}
 Pose_29 {type=Curve, loc=<0 0 0>, rot=<0 0 0 1>}
 Pose_30 {type=Curve, loc=<0 0 0>, rot=<0 0 0 1>}
 Pose_31 {type=Curve, loc=<0 0 0>, rot=<0 0 0 1>}
 Pose_32 {type=Curve, loc=<0 0 0>, rot=<0 0 0 1>}
 Pose_33 {type=Curve, loc=<0 0 0>, rot=<0 0 0 1>}
 Pose_34 {type=Curve, loc=<0 0 0>, rot=<0 0 0 1>}
 Pose_35 {type=Curve, loc=<0 0 0>, rot=<0 0 0 1>}
 Pose_36 {type=Curve, loc=<0 0 0>, rot=<0 0 0 1>}
 Pose_37 {type=Curve, loc=<0 0 0>, rot=<0 0 0 1>}
 Pose_38 {type=Curve, loc=<0 0 0>, rot=<0 0 0 1>}
 Pose_39 {type=Curve, loc=<0 0 0>, rot=<0 0 0 1>}
 Pose_40 {type=Curve, loc=<0 0 0>, rot=<0 0 0 1>}
 Pose_41 {type=Curve, loc=<0 0 0>, rot=<0 0 0 1>}
 Pose_42 {type=Curve, loc=<0 0 0>, rot=<0 0 0 1>}
 Pose_43 {type=Curve, loc=<0 0 0>, rot=<0 0 0 1>}
 Pose_44 {type=Curve, loc=<0 0 0>, rot=<0 0 0 1>}
 Pose_45 {type=Curve, loc=<0 0 0>, rot=<0 0 0 1>}
 Pose_46 {type=Curve, loc=<0 0 0>, rot=<0 0 0 1>}
 Pose_47 {type=Curve, loc=<0 0 0>, rot=<0 0 0 1>}
 Pose_48 {type=Curve, loc=<0 0 0>, rot=<0 0 0 1>}
 Pose_49 {type=Curve, loc=<0 0 0>, rot=<0 0 0 1>}
 Pose_50 {type=Curve, loc=<0 0 0>, rot=<0 0 0 1>}
 Pose_51 {type=Curve, loc=<0 0 0>, rot=<0 0 0 1>}
 Pose_52 {type=Curve, loc=<0 0 0>, rot=<0 0 0 1>}
 Pose_53 {type=Curve, loc=<0 0 0>, rot=<0 0 0 1>}
 Pose_54 {type=Curve, loc=<0 0 0>, rot=<0 0 0 1>}
 Pose_55 {type=Curve, loc=<0 0 0>, rot=<0 0 0 1>}
 Pose_56 {type=Curve, loc=<0 0 0>, rot=<0 0 0 1>}
 Pose_57 {type=Curve, loc=<0 0 0>, rot=<0 0 0 1>}
 Pose_58 {type=Curve, loc=<0 0 0>, rot=<0 0 0 1>}
 Pose_59 {type=Curve, loc=<0 0 0>, rot=<0 0 0 1>}
 Pose_60 {type=Curve, loc=<0 0 0>, rot=<0 0 0 1>}
 Pose_61 {type=Curve, loc=<0 0 0>, rot=<0 0 0 1>}
 Pose_62 {type=Curve, loc=<0 0 0>, rot=<0 0 0 1>}
 Pose_63 {type=Curve, loc=<0 0 0>, rot=<0 0 0 1>}
 Pose_64 {type=Curve, loc=<0 0 0>, rot=<0 0 0 1>}
 Pose_65 {type=Curve, loc=<0 0 0>, rot=<0 0 0 1>}
 Pose_66 {type=Curve, loc=<0 0 0>, rot=<0 0 0 1>}
 Pose_67 {type=Curve, loc=<0 0 0>, rot=<0 0 0 1>}
 Pose_68 {type=Curve, loc=<0 0 0>, rot=<0 0 0 1>}
 Pose_69 {type=Curve, loc=<0 0 0>, rot=<0 0 0 1>}
 Pose_70 {type=Curve, loc=<0 0 0>, rot=<0 0 0 1>}
 Pose_71 {type=Curve, loc=<0 0 0>, rot=<0 0 0 1>}
 Pose_72 {type=Curve, loc=<0 0 0>, rot=<0 0 0 1>}
 Pose_73 {type=Curve, loc=<0 0 0>, rot=<0 0 0 1>}
 Pose_74 {type=Curve, loc=<0 0 0>, rot=<0 0 0 1>}
 Pose_75 {type=Curve, loc=<0 0 0>, rot=<0 0 0 1>}
 Pose_76 {type=Curve, loc=<0 0 0>, rot=<0 0 0 1>}
 Pose_77 {type=Curve, loc=<0 0 0>, rot=<0 0 0 1>}
 Pose_78 {type=Curve, loc=<0 0 0>, rot=<0 0 0 1>}
 Pose_79 {type=Curve, loc=<0 0 0>, rot=<0 0 0 1>}
 Pose_80 {type=Curve, loc=<0 0 0>, rot=<0 0 0 1>}
 Pose_81 {type=Curve, loc=<0 0 0>, rot=<0 0 0 1>}
 Pose_82 {type=Curve, loc=<0 0 0>, rot=<0 0 0 1>}
 Pose_83 {type=Curve, loc=<0 0 0>, rot=<0 0 0 1>}
 Pose_84 {type=Curve, loc=<0 0 0>, rot=<0 0 0 1>}
 Pose_85 {type=Curve, loc=<0 0 0>, rot=<0 0 0 1>}
 Pose_86 {type=Curve, loc=<0 0 0>, rot=<0 0 0 1>}
 Pose_87 {type=Curve, loc=<0 0 0>, rot=<0 0 0 1>}
 Pose_88 {type=Curve, loc=<0 0 0>, rot=<0 0 0 1>}
 Pose_89 {type=Curve, loc=<0 0 0>, rot=<0 0 0 1>}
 Pose_90 {type=Curve, loc=<0 0 0>, rot=<0 0 0 1>}
 Pose_91 {type=Curve, loc=<0 0 0>, rot=<0 0 0 1>}
 Pose_92 {type=Curve, loc=<0 0 0>, rot=<0 0 0 1>}
 Pose_93 {type=Curve, loc=<0 0 0>, rot=<0 0 0 1>}
 Pose_94 {type=Curve, loc=<0 0 0>, rot=<0 0 0 1>}
 Pose_95 {type=Curve, loc=<0 0 0>, rot=<0 0 0 1>}
 Pose_96 {type=Curve, loc=<0 0 0>, rot=<0 0 0 1>}
 Pose_97 {type=Curve, loc=<0 0 0>, rot=<0 0 0 1>}
 Pose_98 {type=Curve, loc=<0 0 0>, rot=<0 0 0 1>}
 Pose_99 {type=Curve, loc=<0 0 0>, rot=<0 0 0 1>}
 Pose_100 {type=Curve, loc=<0 0 0>, rot=<0 0 0 1>}
 Pose_101 {type=Curve, loc=<0 0 0>, rot=<0 0 0 1>}
 Pose_102 {type=Curve, loc=<0 0 0>, rot=<0 0 0 1>}
 Pose_103 {type=Curve, loc=<0 0 0>, rot=<0 0 0 1>}
 Pose_104 {type=Curve, loc=<0 0 0>, rot=<0 0 0 1>}
 Pose_105 {type=Curve, loc=<0 0 0>, rot=<0 0 0 1>}
 Pose_106 {type=Curve, loc=<0 0 0>, rot=<0 0 0 1>}
 Pose_107 {type=Curve, loc=<0 0 0>, rot=<0 0 0 1>}
 Pose_108 {type=Curve, loc=<0 0 0>, rot=<0 0 0 1>}
 Pose_109 {type=Curve, loc=<0 0 0>, rot=<0 0 0 1>}
 Pose_110 {type=Curve, loc=<0 0 0>, rot=<0 0 0 1>}
 Pose_111 {type=Curve, loc=<0 0 0>, rot=<0 0 0 1>}
 Pose_112 {type=Curve, loc=<0 0 0>, rot=<0 0 0 1>}
 Pose_113 {type=Curve, loc=<0 0 0>, rot=<0 0 0 1>}
 Pose_114 {type=Curve, loc=<0 0 0>, rot=<0 0 0 1>}
 Pose_115 {type=Curve, loc=<0 0 0>, rot=<0 0 0 1>}
 Pose_116 {type=Curve, loc=<0 0 0>, rot=<0 0 0 1>}
 Pose_117 {type=Curve, loc=<0 0 0>, rot=<0 0 0 1>}
 Pose_118 {type=Curve, loc=<0 0 0>, rot=<0 0 0 1>}
 Pose_119 {type=Curve, loc=<0 0 0>, rot=<0 0 0 1>}
 Pose_120 {type=Curve, loc=<0 0 0>, rot=<0 0 0 1>}
 Pose_121 {type=Curve, loc=<0 0 0>, rot=<0 0 0 1>}
 Pose_122 {type=Curve, loc=<0 0 0>, rot=<0 0 0 1>}
 Pose_123 {type=Curve, loc=<0 0 0>, rot=<0 0 0 1>}
 Pose_124 {type=Curve, loc=<0 0 0>, rot=<0 0 0 1>}
 Pose_125 {type=Curve, loc=<0 0 0>, rot=<0 0 0 1>}
 Pose_126 {type=Curve, loc=<0 0 0>, rot=<0 0 0 1>}
 Pose_127 {type=Curve, loc=<0 0 0>, rot=<0 0 0 1>}
 Pose_128 {type=Curve, loc=<0 0 0>, rot=<0 0 0 1>}
 Pose_129 {type=Curve, loc=<0 0 0>, rot=<0 0 0 1>}
 Pose_130 {type=Curve, loc=<0 0 0>, rot=<0 0 0 1>}
 Pose_131 {type=Curve, loc=<0 0 0>, rot=<0 0 0 1>}
 Pose_132 {type=Curve, loc=<0 0 0>, rot=<0 0 0 1>}
 Pose_133 {type=Curve, loc=<0 0 0>, rot=<0 0 0 1>}
 Pose_134 {type=Curve, loc=<0 0 0>, rot=<0 0 0 1>}
 Pose_135 {type=Curve, loc=<0 0 0>, rot=<0 0 0 1>}
 Pose_136 {type=Curve, loc=<0 0 0>, rot=<0 0 0 1>}
 Pose_137 {type=Curve, loc=<0 0 0>, rot=<0 0 0 1>}
 Pose_138 {type=Curve, loc=<0 0 0>, rot=<0 0 0 1>}
 Pose_139 {type=Curve, loc=<0 0 0>, rot=<0 0 0 1>}
 Pose_140 {type=Curve, loc=<0 0 0>, rot=<0 0 0 1>}
 Pose_141 {type=Curve, loc=<0 0 0>, rot=<0 0 0 1>}
 Pose_142 {type=Curve, loc=<0 0 0>, rot=<0 0 0 1>}
 Pose_143 {type=Curve, loc=<0 0 0>, rot=<0 0 0 1>}
 Pose_144 {type=Curve, loc=<0 0 0>, rot=<0 0 0 1>}
 Pose_145 {type=Curve, loc=<0 0 0>, rot=<0 0 0 1>}
 Pose_146 {type=Curve, loc=<0 0 0>, rot=<0 0 0 1>}
 Pose_147 {type=Curve, loc=<0 0 0>, rot=<0 0 0 1>}
 Pose_148 {type=Curve, loc=<0 0 0>, rot=<0 0 0 1>}
 Pose_149 {type=Curve, loc=<0 0 0>, rot=<0 0 0 1>}
 Pose_150 {type=Curve, loc=<0 0 0>, rot=<0 0 0 1>}
 Pose_151 {type=Curve, loc=<0 0 0>, rot=<0 0 0 1>}
 Pose_152 {type=Curve, loc=<0 0 0>, rot=<0 0 0 1>}
 Pose_153 {type=Curve, loc=<0 0 0>, rot=<0 0 0 1>}
 Pose_154 {type=Curve, loc=<0 0 0>, rot=<0 0 0 1>}
 Pose_155 {type=Curve, loc=<0 0 0>, rot=<0 0 0 1>}
 Pose_156 {type=Curve, loc=<0 0 0>, rot=<0 0 0 1>}
 Pose_157 {type=Curve, loc=<0 0 0>, rot=<0 0 0 1>}
 Pose_158 {type=Curve, loc=<0 0 0>, rot=<0 0 0 1>}
 Pose_159 {type=Curve, loc=<0 0 0>, rot=<0 0 0 1>}
 Pose_160 {type=Curve, loc=<0 0 0>, rot=<0 0 0 1>}
 Pose_161 {type=Curve, loc=<0 0 0>, rot=<0 0 0 1>}
 Pose_162 {type=Curve, loc=<0 0 0>, rot=<0 0 0 1>}
 Pose_163 {type=Curve, loc=<0 0 0>, rot=<0 0 0 1>}
 Pose_164 {type=Curve, loc=<0 0 0>, rot=<0 0 0 1>}
 Pose_165 {type=Curve, loc=<0 0 0>, rot=<0 0 0 1>}
 Pose_166 {type=Curve, loc=<0 0 0>, rot=<0 0 0 1>}
 Pose_167 {type=Curve, loc=<0 0 0>, rot=<0 0 0 1>}
 Pose_168 {type=Curve, loc=<0 0 0>, rot=<0 0 0 1>}
 Pose_169 {type=Curve, loc=<0 0 0>, rot=<0 0 0 1>}
 Pose_170 {type=Curve, loc=<0 0 0>, rot=<0 0 0 1>}
 Pose_171 {type=Curve, loc=<0 0 0>, rot=<0 0 0 1>}
 Pose_172 {type=Curve, loc=<0 0 0>, rot=<0 0 0 1>}
 Pose_173 {type=Curve, loc=<0 0 0>, rot=<0 0 0 1>}
 Pose_174 {type=Curve, loc=<0 0 0>, rot=<0 0 0 1>}
 Pose_175 {type=Curve, loc=<0 0 0>, rot=<0 0 0 1>}
 Pose_176 {type=Curve, loc=<0 0 0>, rot=<0 0 0 1>}
 Pose_177 {type=Curve, loc=<0 0 0>, rot=<0 0 0 1>}
 Pose_178 {type=Curve, loc=<0 0 0>, rot=<0 0 0 1>}
 Pose_179 {type=Curve, loc=<0 0 0>, rot=<0 0 0 1>}
 eyeLookDownRight {type=Curve, loc=<0 0 0>, rot=<0 0 0 1>}
 eyeLookOutRight {type=Curve, loc=<0 0 0>, rot=<0 0 0 1>}
 eyeLookUpRight {type=Curve, loc=<0 0 0>, rot=<0 0 0 1>}
 eyeLookOutLeft {type=Curve, loc=<0 0 0>, rot=<0 0 0 1>}
 eyeBlinkLeft {type=Curve, loc=<0 0 0>, rot=<0 0 0 1>}
 eyeBlinkRight {type=Curve, loc=<0 0 0>, rot=<0 0 0 1>}
 eyeSquintLeft {type=Curve, loc=<0 0 0>, rot=<0 0 0 1>}
 eyeSquintRight {type=Curve, loc=<0 0 0>, rot=<0 0 0 1>}
 eyeWideLeft {type=Curve, loc=<0 0 0>, rot=<0 0 0 1>}
 eyeWideRight {type=Curve, loc=<0 0 0>, rot=<0 0 0 1>}
 jawForward {type=Curve, loc=<0 0 0>, rot=<0 0 0 1>}
 jawLeft {type=Curve, loc=<0 0 0>, rot=<0 0 0 1>}
 jawRight {type=Curve, loc=<0 0 0>, rot=<0 0 0 1>}
 jawOpen {type=Curve, loc=<0 0 0>, rot=<0 0 0 1>}
 mouthClose {type=Curve, loc=<0 0 0>, rot=<0 0 0 1>}
 mouthFunnel {type=Curve, loc=<0 0 0>, rot=<0 0 0 1>}
 mouthPucker {type=Curve, loc=<0 0 0>, rot=<0 0 0 1>}
 mouthLeft {type=Curve, loc=<0 0 0>, rot=<0 0 0 1>}
 mouthRight {type=Curve, loc=<0 0 0>, rot=<0 0 0 1>}
 mouthSmileLeft {type=Curve, loc=<0 0 0>, rot=<0 0 0 1>}
 mouthSmileRight {type=Curve, loc=<0 0 0>, rot=<0 0 0 1>}
 mouthFrownLeft {type=Curve, loc=<0 0 0>, rot=<0 0 0 1>}
 mouthFrownRight {type=Curve, loc=<0 0 0>, rot=<0 0 0 1>}
 mouthDimpleLeft {type=Curve, loc=<0 0 0>, rot=<0 0 0 1>}
 mouthDimpleRight {type=Curve, loc=<0 0 0>, rot=<0 0 0 1>}
 mouthStretchLeft {type=Curve, loc=<0 0 0>, rot=<0 0 0 1>}
 mouthStretchRight {type=Curve, loc=<0 0 0>, rot=<0 0 0 1>}
 mouthRollLower {type=Curve, loc=<0 0 0>, rot=<0 0 0 1>}
 mouthRollUpper {type=Curve, loc=<0 0 0>, rot=<0 0 0 1>}
 mouthShrugLower {type=Curve, loc=<0 0 0>, rot=<0 0 0 1>}
 mouthShrugUpper {type=Curve, loc=<0 0 0>, rot=<0 0 0 1>}
 mouthPressLeft {type=Curve, loc=<0 0 0>, rot=<0 0 0 1>}
 mouthPressRight {type=Curve, loc=<0 0 0>, rot=<0 0 0 1>}
 mouthLowerDownLeft {type=Curve, loc=<0 0 0>, rot=<0 0 0 1>}
 mouthLowerDownRight {type=Curve, loc=<0 0 0>, rot=<0 0 0 1>}
 mouthUpperUpLeft {type=Curve, loc=<0 0 0>, rot=<0 0 0 1>}
 mouthUpperUpRight {type=Curve, loc=<0 0 0>, rot=<0 0 0 1>}
 browDownLeft {type=Curve, loc=<0 0 0>, rot=<0 0 0 1>}
 browDownRight {type=Curve, loc=<0 0 0>, rot=<0 0 0 1>}
 browInnerUp {type=Curve, loc=<0 0 0>, rot=<0 0 0 1>}
 browOuterUpLeft {type=Curve, loc=<0 0 0>, rot=<0 0 0 1>}
 browOuterUpRight {type=Curve, loc=<0 0 0>, rot=<0 0 0 1>}
 cheekPuff {type=Curve, loc=<0 0 0>, rot=<0 0 0 1>}
 cheekSquintLeft {type=Curve, loc=<0 0 0>, rot=<0 0 0 1>}
 cheekSquintRight {type=Curve, loc=<0 0 0>, rot=<0 0 0 1>}
 noseSneerLeft {type=Curve, loc=<0 0 0>, rot=<0 0 0 1>}
 noseSneerRight {type=Curve, loc=<0 0 0>, rot=<0 0 0 1>}
 C_glabella_down {type=Curve, loc=<0 0 0>, rot=<0 0 0 1>}
 L_brow_squeeze_in {type=Curve, loc=<0 0 0>, rot=<0 0 0 1>}
 R_brow_squeeze_in {type=Curve, loc=<0 0 0>, rot=<0 0 0 1>}
 L_upper_lip_lower {type=Curve, loc=<0 0 0>, rot=<0 0 0 1>}
 R_upper_lip_lower {type=Curve, loc=<0 0 0>, rot=<0 0 0 1>}
 L_lower_lip_up {type=Curve, loc=<0 0 0>, rot=<0 0 0 1>}
 R_lower_lip_up {type=Curve, loc=<0 0 0>, rot=<0 0 0 1>}
 L_lip_corner_narrow {type=Curve, loc=<0 0 0>, rot=<0 0 0 1>}
 R_lip_corner_narrow {type=Curve, loc=<0 0 0>, rot=<0 0 0 1>}
 phoneme_ch {type=Curve, loc=<0 0 0>, rot=<0 0 0 1>}
 phoneme_mbp {type=Curve, loc=<0 0 0>, rot=<0 0 0 1>}
 phoneme_fv {type=Curve, loc=<0 0 0>, rot=<0 0 0 1>}
 mouth_up {type=Curve, loc=<0 0 0>, rot=<0 0 0 1>}
 mouth_down {type=Curve, loc=<0 0 0>, rot=<0 0 0 1>}
 L_cheek_up {type=Curve, loc=<0 0 0>, rot=<0 0 0 1>}
 R_cheek_up {type=Curve, loc=<0 0 0>, rot=<0 0 0 1>}
 L_upperLidDown {type=Curve, loc=<0 0 0>, rot=<0 0 0 1>}
 R_upperLidDown {type=Curve, loc=<0 0 0>, rot=<0 0 0 1>}
 jaw_up {type=Curve, loc=<0 0 0>, rot=<0 0 0 1>}
 jaw_back {type=Curve, loc=<0 0 0>, rot=<0 0 0 1>}
 tongue_forward {type=Curve, loc=<0 0 0>, rot=<0 0 0 1>}
 tongue_back {type=Curve, loc=<0 0 0>, rot=<0 0 0 1>}
 tongue_up {type=Curve, loc=<0 0 0>, rot=<0 0 0 1>}
 tongue_down {type=Curve, loc=<0 0 0>, rot=<0 0 0 1>}
 basePose {type=Curve, loc=<0 0 0>, rot=<0 0 0 1>}
 Pose_180 {type=Curve, loc=<0 0 0>, rot=<0 0 0 1>}
 Pose_181 {type=Curve, loc=<0 0 0>, rot=<0 0 0 1>}
 Pose_182 {type=Curve, loc=<0 0 0>, rot=<0 0 0 1>}
 Pose_183 {type=Curve, loc=<0 0 0>, rot=<0 0 0 1>}
 Pose_184 {type=Curve, loc=<0 0 0>, rot=<0 0 0 1>}
 Pose_185 {type=Curve, loc=<0 0 0>, rot=<0 0 0 1>}
 Pose_186 {type=Curve, loc=<0 0 0>, rot=<0 0 0 1>}
 Pose_187 {type=Curve, loc=<0 0 0>, rot=<0 0 0 1>}
 Pose_188 {type=Curve, loc=<0 0 0>, rot=<0 0 0 1>}
 Pose_189 {type=Curve, loc=<0 0 0>, rot=<0 0 0 1>}
 Pose_190 {type=Curve, loc=<0 0 0>, rot=<0 0 0 1>}
 Pose_191 {type=Curve, loc=<0 0 0>, rot=<0 0 0 1>}
 Pose_192 {type=Curve, loc=<0 0 0>, rot=<0 0 0 1>}
 Pose_193 {type=Curve, loc=<0 0 0>, rot=<0 0 0 1>}
 Pose_194 {type=Curve, loc=<0 0 0>, rot=<0 0 0 1>}
 Pose_195 {type=Curve, loc=<0 0 0>, rot=<0 0 0 1>}
 Pose_196 {type=Curve, loc=<0 0 0>, rot=<0 0 0 1>}
 Pose_197 {type=Curve, loc=<0 0 0>, rot=<0 0 0 1>}
 Pose_198 {type=Curve, loc=<0 0 0>, rot=<0 0 0 1>}
 Pose_199 {type=Curve, loc=<0 0 0>, rot=<0 0 0 1>}
 Pose_200 {type=Curve, loc=<0 0 0>, rot=<0 0 0 1>}
 Pose_201 {type=Curve, loc=<0 0 0>, rot=<0 0 0 1>}
 Pose_202 {type=Curve, loc=<0 0 0>, rot=<0 0 0 1>}
 Pose_203 {type=Curve, loc=<0 0 0>, rot=<0 0 0 1>}
 Pose_204 {type=Curve, loc=<0 0 0>, rot=<0 0 0 1>}
 Pose_205 {type=Curve, loc=<0 0 0>, rot=<0 0 0 1>}
 Pose_206 {type=Curve, loc=<0 0 0>, rot=<0 0 0 1>}
 Pose_207 {type=Curve, loc=<0 0 0>, rot=<0 0 0 1>}
 Pose_208 {type=Curve, loc=<0 0 0>, rot=<0 0 0 1>}
 Pose_209 {type=Curve, loc=<0 0 0>, rot=<0 0 0 1>}
 Pose_210 {type=Curve, loc=<0 0 0>, rot=<0 0 0 1>}
 Pose_211 {type=Curve, loc=<0 0 0>, rot=<0 0 0 1>}
 Pose_212 {type=Curve, loc=<0 0 0>, rot=<0 0 0 1>}
 Pose_213 {type=Curve, loc=<0 0 0>, rot=<0 0 0 1>}
 Pose_214 {type=Curve, loc=<0 0 0>, rot=<0 0 0 1>}
 Pose_215 {type=Curve, loc=<0 0 0>, rot=<0 0 0 1>}
 Pose_216 {type=Curve, loc=<0 0 0>, rot=<0 0 0 1>}
 Pose_217 {type=Curve, loc=<0 0 0>, rot=<0 0 0 1>}
 Pose_218 {type=Curve, loc=<0 0 0>, rot=<0 0 0 1>}
 Pose_219 {type=Curve, loc=<0 0 0>, rot=<0 0 0 1>}
 Pose_220 {type=Curve, loc=<0 0 0>, rot=<0 0 0 1>}
 Pose_221 {type=Curve, loc=<0 0 0>, rot=<0 0 0 1>}
 Pose_222 {type=Curve, loc=<0 0 0>, rot=<0 0 0 1>}
 Pose_223 {type=Curve, loc=<0 0 0>, rot=<0 0 0 1>}
 Pose_224 {type=Curve, loc=<0 0 0>, rot=<0 0 0 1>}
 Pose_225 {type=Curve, loc=<0 0 0>, rot=<0 0 0 1>}
 Pose_226 {type=Curve, loc=<0 0 0>, rot=<0 0 0 1>}
 Pose_227 {type=Curve, loc=<0 0 0>, rot=<0 0 0 1>}
 Pose_228 {type=Curve, loc=<0 0 0>, rot=<0 0 0 1>}
 Pose_229 {type=Curve, loc=<0 0 0>, rot=<0 0 0 1>}
 Pose_230 {type=Curve, loc=<0 0 0>, rot=<0 0 0 1>}
 Pose_231 {type=Curve, loc=<0 0 0>, rot=<0 0 0 1>}
 Pose_232 {type=Curve, loc=<0 0 0>, rot=<0 0 0 1>}
 Pose_233 {type=Curve, loc=<0 0 0>, rot=<0 0 0 1>}
 Pose_234 {type=Curve, loc=<0 0 0>, rot=<0 0 0 1>}
 Pose_235 {type=Curve, loc=<0 0 0>, rot=<0 0 0 1>}
 Pose_236 {type=Curve, loc=<0 0 0>, rot=<0 0 0 1>}
 Pose_237 {type=Curve, loc=<0 0 0>, rot=<0 0 0 1>}
 Pose_238 {type=Curve, loc=<0 0 0>, rot=<0 0 0 1>}
 Pose_239 {type=Curve, loc=<0 0 0>, rot=<0 0 0 1>}
 Pose_240 {type=Curve, loc=<0 0 0>, rot=<0 0 0 1>}
 Pose_241 {type=Curve, loc=<0 0 0>, rot=<0 0 0 1>}
 Pose_242 {type=Curve, loc=<0 0 0>, rot=<0 0 0 1>}
 Pose_243 {type=Curve, loc=<0 0 0>, rot=<0 0 0 1>}
 Pose_244 {type=Curve, loc=<0 0 0>, rot=<0 0 0 1>}
 Pose_245 {type=Curve, loc=<0 0 0>, rot=<0 0 0 1>}
 Pose_246 {type=Curve, loc=<0 0 0>, rot=<0 0 0 1>}
 Pose_247 {type=Curve, loc=<0 0 0>, rot=<0 0 0 1>}
 Pose_248 {type=Curve, loc=<0 0 0>, rot=<0 0 0 1>}
 Pose_249 {type=Curve, loc=<0 0 0>, rot=<0 0 0 1>}
 Pose_250 {type=Curve, loc=<0 0 0>, rot=<0 0 0 1>}
 Pose_251 {type=Curve, loc=<0 0 0>, rot=<0 0 0 1>}
 Pose_252 {type=Curve, loc=<0 0 0>, rot=<0 0 0 1>}
 Pose_253 {type=Curve, loc=<0 0 0>, rot=<0 0 0 1>}
 Pose_254 {type=Curve, loc=<0 0 0>, rot=<0 0 0 1>}
 Pose_255 {type=Curve, loc=<0 0 0>, rot=<0 0 0 1>}
 Pose_256 {type=Curve, loc=<0 0 0>, rot=<0 0 0 1>}
 Pose_257 {type=Curve, loc=<0 0 0>, rot=<0 0 0 1>}
 Pose_258 {type=Curve, loc=<0 0 0>, rot=<0 0 0 1>}
 neutral {type=Curve, loc=<0 0 0>, rot=<0 0 0 1>}
 Pose_259 {type=Curve, loc=<0 0 0>, rot=<0 0 0 1>}
 cheek_puff_low {type=Curve, loc=<0 0 0>, rot=<0 0 0 1>}
 Pose_260 {type=Curve, loc=<0 0 0>, rot=<0 0 0 1>}
 Pose_261 {type=Curve, loc=<0 0 0>, rot=<0 0 0 1>}
 Pose_262 {type=Curve, loc=<0 0 0>, rot=<0 0 0 1>}
 Pose_263 {type=Curve, loc=<0 0 0>, rot=<0 0 0 1>}
 Pose_264 {type=Curve, loc=<0 0 0>, rot=<0 0 0 1>}
 Pose_265 {type=Curve, loc=<0 0 0>, rot=<0 0 0 1>}
 Pose_266 {type=Curve, loc=<0 0 0>, rot=<0 0 0 1>}
 Pose_267 {type=Curve, loc=<0 0 0>, rot=<0 0 0 1>}
 Pose_268 {type=Curve, loc=<0 0 0>, rot=<0 0 0 1>}
 Pose_269 {type=Curve, loc=<0 0 0>, rot=<0 0 0 1>}
 Pose_270 {type=Curve, loc=<0 0 0>, rot=<0 0 0 1>}
 Pose_271 {type=Curve, loc=<0 0 0>, rot=<0 0 0 1>}
 Pose_272 {type=Curve, loc=<0 0 0>, rot=<0 0 0 1>}
 Pose_273 {type=Curve, loc=<0 0 0>, rot=<0 0 0 1>}
 Pose_274 {type=Curve, loc=<0 0 0>, rot=<0 0 0 1>}
 Pose_275 {type=Curve, loc=<0 0 0>, rot=<0 0 0 1>}
 Pose_276 {type=Curve, loc=<0 0 0>, rot=<0 0 0 1>}
 Pose_277 {type=Curve, loc=<0 0 0>, rot=<0 0 0 1>}
 Pose_278 {type=Curve, loc=<0 0 0>, rot=<0 0 0 1>}
 Pose_279 {type=Curve, loc=<0 0 0>, rot=<0 0 0 1>}
 Pose_280 {type=Curve, loc=<0 0 0>, rot=<0 0 0 1>}
 Pose_281 {type=Curve, loc=<0 0 0>, rot=<0 0 0 1>}
 Pose_282 {type=Curve, loc=<0 0 0>, rot=<0 0 0 1>}
 Pose_283 {type=Curve, loc=<0 0 0>, rot=<0 0 0 1>}
 Pose_284 {type=Curve, loc=<0 0 0>, rot=<0 0 0 1>}
 Pose_285 {type=Curve, loc=<0 0 0>, rot=<0 0 0 1>}
 Pose_286 {type=Curve, loc=<0 0 0>, rot=<0 0 0 1>}
 Pose_287 {type=Curve, loc=<0 0 0>, rot=<0 0 0 1>}
 Pose_288 {type=Curve, loc=<0 0 0>, rot=<0 0 0 1>}
 Pose_289 {type=Curve, loc=<0 0 0>, rot=<0 0 0 1>}
 Pose_290 {type=Curve, loc=<0 0 0>, rot=<0 0 0 1>}
 Pose_291 {type=Curve, loc=<0 0 0>, rot=<0 0 0 1>}
 Pose_292 {type=Curve, loc=<0 0 0>, rot=<0 0 0 1>}
 Pose_293 {type=Curve, loc=<0 0 0>, rot=<0 0 0 1>}
 Pose_294 {type=Curve, loc=<0 0 0>, rot=<0 0 0 1>}
 Pose_295 {type=Curve, loc=<0 0 0>, rot=<0 0 0 1>}
 Pose_296 {type=Curve, loc=<0 0 0>, rot=<0 0 0 1>}
 Pose_297 {type=Curve, loc=<0 0 0>, rot=<0 0 0 1>}
 Pose_298 {type=Curve, loc=<0 0 0>, rot=<0 0 0 1>}
 Pose_299 {type=Curve, loc=<0 0 0>, rot=<0 0 0 1>}
 Pose_300 {type=Curve, loc=<0 0 0>, rot=<0 0 0 1>}
 Pose_301 {type=Curve, loc=<0 0 0>, rot=<0 0 0 1>}
 Pose_302 {type=Curve, loc=<0 0 0>, rot=<0 0 0 1>}
 Pose_303 {type=Curve, loc=<0 0 0>, rot=<0 0 0 1>}
 Pose_304 {type=Curve, loc=<0 0 0>, rot=<0 0 0 1>}
 Pose_305 {type=Curve, loc=<0 0 0>, rot=<0 0 0 1>}
 Pose_306 {type=Curve, loc=<0 0 0>, rot=<0 0 0 1>}
 Pose_307 {type=Curve, loc=<0 0 0>, rot=<0 0 0 1>}
 Pose_308 {type=Curve, loc=<0 0 0>, rot=<0 0 0 1>}
 Pose_309 {type=Curve, loc=<0 0 0>, rot=<0 0 0 1>}
 Pose_310 {type=Curve, loc=<0 0 0>, rot=<0 0 0 1>}
 Pose_311 {type=Curve, loc=<0 0 0>, rot=<0 0 0 1>}
 Pose_312 {type=Curve, loc=<0 0 0>, rot=<0 0 0 1>}
 Pose_313 {type=Curve, loc=<0 0 0>, rot=<0 0 0 1>}
 Pose_314 {type=Curve, loc=<0 0 0>, rot=<0 0 0 1>}
 Pose_315 {type=Curve, loc=<0 0 0>, rot=<0 0 0 1>}
 Pose_316 {type=Curve, loc=<0 0 0>, rot=<0 0 0 1>}
 Pose_317 {type=Curve, loc=<0 0 0>, rot=<0 0 0 1>}
 Pose_318 {type=Curve, loc=<0 0 0>, rot=<0 0 0 1>}
 Pose_319 {type=Curve, loc=<0 0 0>, rot=<0 0 0 1>}
 Pose_320 {type=Curve, loc=<0 0 0>, rot=<0 0 0 1>}
 Pose_321 {type=Curve, loc=<0 0 0>, rot=<0 0 0 1>}
 Desert_Island_Jonesy_Mouth {type=Curve, loc=<0 0 0>, rot=<0 0 0 1>}
 Pose_322 {type=Curve, loc=<0 0 0>, rot=<0 0 0 1>}
 HT_L1_TransitionCurve {type=Curve, loc=<0 0 0>, rot=<0 0 0 1>}
 Hair_CapShape {type=Curve, loc=<0 0 0>, rot=<0 0 0 1>}
 Hair_CapShape1 {type=Curve, loc=<0 0 0>, rot=<0 0 0 1>}
 jekyll {type=Curve, loc=<0 0 0>, rot=<0 0 0 1>}
 JH_TransformCurve {type=Curve, loc=<0 0 0>, rot=<0 0 0 1>}
 eyes_LOD0 {type=Curve, loc=<0 0 0>, rot=<0 0 0 1>}
 MaterialCurveA {type=Curve, loc=<0 0 0>, rot=<0 0 0 1>}
 MaterialCurveB {type=Curve, loc=<0 0 0>, rot=<0 0 0 1>}
 MaterialCurveC {type=Curve, loc=<0 0 0>, rot=<0 0 0 1>}
 MaterialCurveD {type=Curve, loc=<0 0 0>, rot=<0 0 0 1>}
 MaterialCurveE {type=Curve, loc=<0 0 0>, rot=<0 0 0 1>}
]
